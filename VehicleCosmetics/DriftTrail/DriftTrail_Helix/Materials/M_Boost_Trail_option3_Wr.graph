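# Unlit masked Niagara material for the helix drift/boost trail cosmetic, usable on ribbons, sprites and mesh particles.
# A panned trail gradient and wind-particle streaks, broken up by two scrolling noise samples, are shaped by UV end and top/bottom fades.
# Dynamic parameters blend built-in green toward particle color, scale brightness and hue-shift a sphere-masked base tint.
# The summed color feeds an emissive-adjustment function; particle alpha scales edge and streak masks.
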Material /VehicleCosmetics/DriftTrail/DriftTrail_Helix/Materials/M_Boost_Trail_option3_Wr
graph {
 properties {
  bAutomaticallySetUsageInEditor=false
  bUsedWithNiagaraMeshParticles=true
  bUsedWithNiagaraRibbons=true
  bUsedWithNiagaraSprites=true
  blendMode=BLEND_Masked
  shadingModel=MSM_Unlit
  usedShadingModels=Unlit
 }
 nodes [
  n1 = TextureCoordinate()
  n2 = Multiply(A=Time(), B=ScalarParameter() {defaultValue=1})
  n3 = Multiply(A=TextureSample(UVs=Panner(Coordinate=@n1, Time=@n2) {speedX=1}) {texture=/DealershipSharedLibrary/FX/Textures/Gradient_Masks/T_VCVFX_Trail3_Alpha}.R) {constB=0}
  n4 = TextureCoordinate()
  n5 = Power(Base=OneMinus(None=ComponentMask(None=@n4) {r=true}), Exp=Constant() {r=10})
  n6 = Constant() {r=1}
  n7 = Power(Base=ComponentMask(None=TextureCoordinate()) {r=true}, Exp=Constant() {r=2})
  n8 = Constant() {r=500}
  MaterialExpressionConstantClamp_8 = Saturate(None=Multiply(A=Multiply(A=@n5, B=@n6), B=Multiply(A=@n7, B=@n8)))
  MaterialExpressionConstantClamp_7 = Saturate(None=Multiply(A=@n3, B=@MaterialExpressionConstantClamp_8))
  MaterialExpressionVertexColor_1 = ParticleColor()
  n9 = Constant() {r=20}
  n10 = Power(
   Base=MaterialFunctionCall!(
    "Gradient U=0 V=1"=Constant() {r=1}
    "Invert gradient 0=False 1=True"
    "Contrast (Power)"=ScalarParameter() {defaultValue=2.8, group=Opacity, parameterName=Contrast Fade Top, sortPriority=9}
    "Brightness (Multiply)"=ScalarParameter() {defaultValue=100, group=Opacity, parameterName=Brightness Fade Top, sortPriority=10}
    "Dynamic Contrast (Power)"
    "Dynamic Brightness (Multiply)"
   ) {materialFunction=/DealershipSharedLibrary/FX/Materials/MaterialFunctions/MF_LinearGradientPlus}.Result
   Exp=@n9
  ) {constExponent=20}
  n11 = Constant() {r=1}
  n12 = Power(
   Base=MaterialFunctionCall(
    "Gradient U=0 V=1"=@n11
    "Invert gradient 0=False 1=True"=@n11
    "Contrast (Power)"=ScalarParameter() {defaultValue=2.8, group=Opacity, parameterName=Contrast Fade Bottom, sortPriority=11}
    "Brightness (Multiply)"=ScalarParameter() {defaultValue=100, group=Opacity, parameterName=Brightness Fade Bottom, sortPriority=12}
   ) {materialFunction=&MF_LinearGradientPlus}.Result
   Exp=@n9
  ) {constExponent=20}
  n13 = Multiply(A=Saturate(None=Multiply(A=@n10, B=@n12)), B=ScalarParameter() {defaultValue=0.8, parameterName=Intensity_Mask}) {constB=0.8}
  n14 = Multiply(A=@MaterialExpressionVertexColor_1.A, B=@n13) {constA=1}
  MaterialExpressionConstantClamp_6 = Saturate(None=Add(A=@MaterialExpressionConstantClamp_7, B=@n14))
  n15 = Constant3Vector() {constant=<rgba 0.212 1 0.0987 0>}
  n16 = Multiply(A=Multiply(A=Multiply(A=@n3, B=@n15), B=ScalarParameter() {defaultValue=1.3, parameterName=Color}))
  n17 = DynamicParameter() {defaultValue=<rgba 1 1 0 0>, paramNames [desaturation,power,HueShift,Param4]}
  n18 = Multiply(A=Multiply(A=@MaterialExpressionVertexColor_1.RGB, B=@n3), B=@n17.Param4)
  n19 = Time()
  n20 = ScalarParameter() {defaultValue=1}
  n21 = TextureSample(UVs=Panner(Coordinate=TextureCoordinate() {uTiling=2, vTiling=0.3}, Time=Multiply(A=@n19, B=@n20)) {speedX=1}) {texture=/DealershipSharedLibrary/FX/Textures/Gradient_Masks/T_VCVFX_Boost_Helix_WindParticles_Alpha}
  n22 = ComponentMask(None=TextureCoordinate()) {r=true}
  n23 = Multiply(A=Power(Base=OneMinus(None=@n22), Exp=Constant() {r=5}), B=Constant() {r=1})
  n24 = Multiply(A=ScalarParameter() {defaultValue=1, parameterName=MaskSpeed}, B=Time())
  n25 = TextureSample(UVs=Panner(Coordinate=TextureCoordinate() {uTiling=2, vTiling=0.25}, Time=@n24) {speedX=-1, speedY=-0.1}) {texture=/VehicleCosmetics/FX/Textures/Noise/T_VCVFX_NoiseStarBeam_A}
  n26 = Constant() {r=1.5}
  n27 = TextureSample(UVs=Panner(Coordinate=TextureCoordinate() {vTiling=0.5}, Time=@n24) {speedX=-1, speedY=-0.1}) {texture=&T_VCVFX_NoiseStarBeam_A}
  n28 = Constant() {r=1.5}
  MaterialExpressionConstantClamp_11 = Saturate(None=Add(A=Multiply(A=@n25.RGB, B=@n26), B=Multiply(A=@n27.RGB, B=@n28)))
  MaterialExpressionConstantClamp_13 = Saturate(None=Add(A=@n23, B=Multiply(A=@n23, B=@MaterialExpressionConstantClamp_11)))
  n29 = Power(Base=ComponentMask(None=TextureCoordinate()) {r=true}, Exp=Constant() {r=1.5})
  n30 = Constant() {r=500}
  MaterialExpressionConstantClamp_12 = Saturate(None=Multiply(A=@MaterialExpressionConstantClamp_13, B=Multiply(A=@n29, B=@n30)))
  n31 = Multiply(A=@n21.R, B=Multiply(A=@n14, B=@MaterialExpressionConstantClamp_12))
  n32 = Constant3Vector() {constant=<rgba 0.2118 1 0.098 0>}
  n33 = Multiply(A=@n13, B=@n21.R)
  n34 = LinearInterpolate(
   A=Multiply(A=Multiply(A=@n31, B=@n32), B=ScalarParameter() {defaultValue=2, parameterName=Manchas})
   B=Multiply(A=Multiply(A=@MaterialExpressionVertexColor_1.RGB, B=@n33), B=@n17.power)
   Alpha=@n17.desaturation
  )
  n35 = DynamicParameter() {defaultValue=<rgba 1 1 0 0>, paramNames [desaturation,power,HueShift,Param4]}
  n36 = Multiply(A=Desaturation(None=ParticleColor().RGB, Fraction=Constant() {r=0.2})) {constB=0.2}
  n37 = Power(Base=SphereMask(A=TextureCoordinate() {vTiling=0}, B=Constant2Vector() {r=0.7}) {attenuationRadius=0.45, hardnessPercent=1}) {constExponent=5}
  n38 = Add(
   A=Multiply(
    A=MaterialFunctionCall(Hue Shift Percentage=@n35.HueShift, Texture=@n36) {materialFunction=&HueShift}.Result
    B=OneMinus(None=@n37)
   )
   B=Multiply(A=@n37, B=@MaterialExpressionVertexColor_1.RGB)
  )
  MaterialExpressionConstantClamp_5 = Saturate(None=Add(A=LinearInterpolate(A=@n16, B=@n18, Alpha=@n17.desaturation), B=Add(A=@n34, B=@n38)))
  MaterialFunctionCall(Emissive=@MaterialExpressionConstantClamp_5) {materialFunction=&MF_FX_EmissiveAdjustments}
 ]
}
